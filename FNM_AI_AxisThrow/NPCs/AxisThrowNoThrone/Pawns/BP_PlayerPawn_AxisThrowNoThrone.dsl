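; BP_PlayerPawn_AxisThrowNoThrone is a player pawn variant of the AxisThrow NPC set for setups without thrones.
; On BeginPlay it runs the parent logic, then destroys the inherited proximity trigger component.
; It overrides the throne collision-disable event to do only the push back.
; On the authority, after a short delay, it activates abilities matching the push-back tag through the ability system component, if that is valid.
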(event EventBeginPlay
  (|CallParentFunction)
  (Components|DestroyComponent (Variables|Default|GetProximityTrigger)))

(event EventTryDisableCollisionsWithThrone
  (bind _returnvalue (Ability|GetAbilitySystemComponent 0))
  ; Do Nothing but keep the Push back, No thrones here!
  (Networking|SwitchHasAuthority
    (:Authority
      (Utilities|FlowControl|Delay 0.2)
      (Utilities|IsValid _returnvalue
        (:Is%20Valid
          (Abilities|TryActivateAbilitiesbyTag _returnvalue (Variables|Default|GetTag_PushBack) false))))))

(fn ConstructionScript
  (|CallParentFunction))
